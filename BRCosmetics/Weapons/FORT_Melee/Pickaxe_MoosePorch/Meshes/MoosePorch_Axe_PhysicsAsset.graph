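# Physics asset for the MoosePorch axe, a two-handed melee pickaxe cosmetic, giving its skeletal mesh a simple physics shape.
# A single capsule body on the Melee_TwoHanded_Handle bone, offset mostly along the bone's Z axis, is the only shape.
# It has no constraints, so the asset defines no jointed or secondary moving parts.
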PhysicsAsset /BRCosmetics/Weapons/FORT_Melee/Pickaxe_MoosePorch/Meshes/MoosePorch_Axe_PhysicsAsset
bodies {Melee_TwoHanded_Handle [{shapeType=Capsule, length=56.4583, radius=13.9943, center=<0.2197 -0.2643 23.6501>, rotation=<rot -0.0244 -90 0.0184>}]}
constraints []
